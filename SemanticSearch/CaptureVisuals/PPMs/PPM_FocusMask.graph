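# PPM_FocusMask is a post-process material that renders a binary mask of pixels matching a chosen custom stencil value, apparently to isolate a focus object in Semantic Search captures.
# It compares the red channel of the CustomStencil scene texture with the TargetStencil parameter, giving 1 on an exact match and 0 otherwise.
# It replaces the tonemapper, so the mask values bypass tonemapping.
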Material /SemanticSearch/CaptureVisuals/PPMs/PPM_FocusMask
graph {
 properties {blendableLocation=BL_ReplacingTonemapper, materialDomain=MD_PostProcess, usedShadingModels=Default Lit}
 nodes [
  n1 = Constant()
  If(
   A=ComponentMask(None=SceneTexture() {sceneTextureId=PPI_CustomStencil}.Color) {r=true}
   B=ScalarParameter() {parameterName=TargetStencil}
   "A > B"=@n1
   "A == B"=Constant() {r=1}
   "A < B"=@n1
  )
 ]
}
