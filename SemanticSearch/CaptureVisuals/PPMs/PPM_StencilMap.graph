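# Post-process material in the SemanticSearch capture visuals that turns the custom stencil buffer into a normalized stencil map.
# It samples the CustomStencil scene texture, takes the red channel and divides it by 255, mapping integer stencil IDs to the 0-1 range.
# It blends into scene color before bloom.
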Material /SemanticSearch/CaptureVisuals/PPMs/PPM_StencilMap
graph {
 properties {blendableLocation=BL_SceneColorBeforeBloom, materialDomain=MD_PostProcess, usedShadingModels=Default Lit}
 nodes [Divide(A=ComponentMask(None=SceneTexture() {sceneTextureId=PPI_CustomStencil}.Color) {r=true}) {constB=255}]
}
